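# Temporary physics asset for the belly of the large male TatToon player body, covering its upper and lower dynamic belly bones.
# Each belly bone has a large offset capsule constrained to a tiny base capsule with angular limits.
# The upper base attaches to spine_02 and the lower to spine_01, which chain to the pelvis through small capsules and angularly free constraints.
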PhysicsAsset /BRCosmetics/Characters/Player/Male/Large/Bodies/M_LRG_TatToon_TEMP/M_LRG_TatToon_Belly_Physics_TEMP
bodies {
 dyn_belly_lower_base_c [{shapeType=Capsule, length=0.505, radius=0.505, shapeName=dyn_belly_lower_base_c_capsule, rotation=<rot 0 -90 0>}]
 dyn_belly_lower_c [{shapeType=Capsule, length=6, radius=14, shapeName=dyn_belly_lower_c_capsule, center=<-0.5326 5.4033 0>}]
 dyn_belly_upper_base_c [{shapeType=Capsule, length=0.505, radius=0.505, shapeName=dyn_belly_upper_base_c_capsule, rotation=<rot 0 -90 0>}]
 dyn_belly_upper_c [{shapeType=Capsule, length=6, radius=14, shapeName=dyn_belly_upper_c_capsule, center=<-0.5326 5.4033 0>}]
 pelvis [{shapeType=Capsule, length=2, radius=2, shapeName=pelvis_capsule}]
 spine_01 [{shapeType=Capsule, length=2, radius=2, shapeName=spine_01_capsule}]
 spine_02 [{shapeType=Capsule, length=2, radius=2, shapeName=spine_02_capsule}]
}
constraints [
 {bone1Name=dyn_belly_upper_c, bone2Name=dyn_belly_upper_base_c, swing1LimitDegrees=40, swing2LimitDegrees=40, twistLimitDegrees=40}
 {bone1Name=dyn_belly_lower_c, bone2Name=dyn_belly_lower_base_c, swing1LimitDegrees=40, swing2LimitDegrees=40, twistLimitDegrees=40}
 {
  bone1Name=dyn_belly_upper_base_c
  bone2Name=spine_02
  swing1LimitDegrees=60
  swing1Motion=Free
  swing2LimitDegrees=60
  swing2Motion=Free
  twistLimitDegrees=60
  twistMotion=Free
 }
 {
  bone1Name=dyn_belly_lower_base_c
  bone2Name=spine_01
  swing1LimitDegrees=60
  swing1Motion=Free
  swing2LimitDegrees=60
  swing2Motion=Free
  twistLimitDegrees=60
  twistMotion=Free
 }
 {
  bone1Name=spine_02
  bone2Name=spine_01
  swing1LimitDegrees=45
  swing1Motion=Free
  swing2LimitDegrees=45
  swing2Motion=Free
  twistLimitDegrees=45
  twistMotion=Free
 }
 {
  bone1Name=spine_01
  bone2Name=pelvis
  swing1LimitDegrees=45
  swing1Motion=Free
  swing2LimitDegrees=45
  swing2Motion=Free
  twistLimitDegrees=45
  twistMotion=Free
 }
]
